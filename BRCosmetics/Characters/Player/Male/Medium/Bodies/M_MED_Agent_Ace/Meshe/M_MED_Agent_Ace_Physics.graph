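# Physics asset for the M_MED_Agent_Ace player body, set up for secondary motion on the right leg-pack strap and left vest strap rather than a full ragdoll.
# Capsules on the pelvis, two spine bones and both thighs form a partial body, linked by angularly free constraints.
# Each strap chain hangs from a small anchor capsule, free-jointed to thigh_r or spine_01, followed by two thin boxes with tight swing and twist limits.
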PhysicsAsset /BRCosmetics/Characters/Player/Male/Medium/Bodies/M_MED_Agent_Ace/Meshe/M_MED_Agent_Ace_Physics
bodies {
 dyn_legpackstrap_a_base_r [{shapeType=Capsule, length=0.505, radius=0.505, rotation=<rot 0 -90 0>}]
 dyn_legpackstrap_a_r [{shapeType=Box, extentX=6, extentY=3.5, extentZ=1, center=<3.1395 0 -0.0002>}]
 dyn_legpackstrap_b_r [{shapeType=Box, extentX=7, extentY=4.5, extentZ=1, center=<3.5694 0 0.0001>}]
 dyn_vest_strap_a_l [{shapeType=Box, extentX=4, extentY=3, extentZ=1, center=<1.758 0 0>}]
 dyn_vest_strap_b_l [{shapeType=Box, extentX=4, extentY=3, extentZ=1, center=<2.0065 0 0>}]
 dyn_vest_strap_constr_l [{shapeType=Capsule, length=1, radius=1}]
 pelvis [{shapeType=Capsule, length=5.1466, radius=12.1752, rotation=<rot 2.0584 -90 1.8798>}]
 spine_01 [{shapeType=Capsule, length=6.25, radius=12.5, center=<0 1.1607 0>}]
 spine_03 [{shapeType=Capsule, length=6.75, radius=11.25, center=<1.1563 0.6329 0>}]
 thigh_l [{shapeType=Capsule, length=19, radius=8.25, center=<-14.1259 0.0001 0>, rotation=<rot 90 0 0>}]
 thigh_r [{shapeType=Capsule, length=19, radius=8.25, center=<14.1259 -0.0001 0>, rotation=<rot 90 0 180>}]
}
constraints [
 {
  bone1Name=spine_03
  bone2Name=spine_01
  swing1LimitDegrees=45
  swing1Motion=Free
  swing2LimitDegrees=45
  swing2Motion=Free
  twistLimitDegrees=45
  twistMotion=Free
 }
 {
  bone1Name=spine_01
  bone2Name=pelvis
  swing1LimitDegrees=45
  swing1Motion=Free
  swing2LimitDegrees=45
  swing2Motion=Free
  twistLimitDegrees=45
  twistMotion=Free
 }
 {
  bone1Name=dyn_legpackstrap_a_base_r
  bone2Name=thigh_r
  swing1LimitDegrees=45
  swing1Motion=Free
  swing2LimitDegrees=45
  swing2Motion=Free
  twistLimitDegrees=45
  twistMotion=Free
 }
 {bone1Name=dyn_legpackstrap_a_r, bone2Name=dyn_legpackstrap_a_base_r, swing1LimitDegrees=10, swing2LimitDegrees=20, twistLimitDegrees=5}
 {bone1Name=dyn_legpackstrap_b_r, bone2Name=dyn_legpackstrap_a_r, swing1LimitDegrees=10, swing2LimitDegrees=20, twistLimitDegrees=5}
 {
  bone1Name=thigh_l
  bone2Name=pelvis
  swing1LimitDegrees=45
  swing1Motion=Free
  swing2LimitDegrees=45
  swing2Motion=Free
  twistLimitDegrees=45
  twistMotion=Free
 }
 {
  bone1Name=thigh_r
  bone2Name=pelvis
  swing1LimitDegrees=45
  swing1Motion=Free
  swing2LimitDegrees=45
  swing2Motion=Free
  twistLimitDegrees=45
  twistMotion=Free
 }
 {
  bone1Name=dyn_vest_strap_constr_l
  bone2Name=spine_01
  swing1LimitDegrees=45
  swing1Motion=Free
  swing2LimitDegrees=45
  swing2Motion=Free
  twistLimitDegrees=45
  twistMotion=Free
 }
 {bone1Name=dyn_vest_strap_a_l, bone2Name=dyn_vest_strap_constr_l, swing1LimitDegrees=20, swing2LimitDegrees=5, twistLimitDegrees=5}
 {bone1Name=dyn_vest_strap_b_l, bone2Name=dyn_vest_strap_a_l, swing1LimitDegrees=20, swing2LimitDegrees=5, twistLimitDegrees=5}
]
