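# Post-process style cosmetic AnimBlueprint for the Agent Ace male medium body, following the attached parent mesh and adding secondary motion to outfit accessories.
# Copy Pose From Mesh feeds the control_rig_ace Control Rig, then weighted multi-bone constraints place zipper, nametag, gun, vest and leg strap base bones.
# Spring-based AnimDynamics swings nametag and zipper, then a pelvis-space RigidBody node simulates the body's physics asset.
# The EventGraph is empty and there are no variables.
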AnimBlueprint /BRCosmetics/Characters/Player/Male/Medium/Bodies/M_MED_Agent_Ace/Meshe/M_MED_Agent_Ace_AnimBP
target_skeleton=&/M_MED_Agent_Ace_Skeleton
graphs [
 AnimGraph {nodes [
   n1 = Control Rig(Source=Copy Pose From Mesh() {bUseAttachedParent=true}) {
    alphaBoolBlend {blendOption=Linear}
    controlRigAssetReference {blueprintRigClass=&/control_rig_ace.control_rig_ace_C}
   }
   n2 = "Constraint - dyn_neck_zipper_base (PPPPPPP)"!(
    ComponentPose=Local To Component(LocalPose=@n1)
    Alpha=1
    ConstraintWeights_0=0.4511
    ConstraintWeights_1:float=0.247
    ConstraintWeights_2:float=0.1647
    ConstraintWeights_3:float=0.0975
    ConstraintWeights_4:float=0.0387
    ConstraintWeights_5:float=0.0005
    ConstraintWeights_6:float=0.0004
   ) -> (Pose) {
    alphaBoolBlend {blendOption=Linear}
    boneToModify {boneName=dyn_neck_zipper_base}
    constraintSetup [
     {targetBone {boneName=spine_05}}
     {targetBone {boneName=neck_02}}
     {targetBone {boneName=head}}
     {targetBone {boneName=neck_01}}
     {targetBone {boneName=spine_04}}
     {targetBone {boneName=clavicle_l}}
     {targetBone {boneName=clavicle_r}}
    ]
    constraintWeights [1,1,1,1,1,1,1]
   }
   n3 = "Constraint - dyn_nametag_constr_base (PPPPPP)"!(
    ComponentPose=@n2
    Alpha=1
    ConstraintWeights_0=0.2031
    ConstraintWeights_1:float=0.0356
    ConstraintWeights_2:float=0.1265
    ConstraintWeights_3:float=0.2667
    ConstraintWeights_4:float=0.2905
    ConstraintWeights_5:float=0.0777
   ) -> (Pose) {
    alphaBoolBlend {blendOption=Linear}
    boneToModify {boneName=dyn_nametag_constr_base}
    constraintSetup [
     {targetBone {boneName=clavicle_l}}
     {targetBone {boneName=spine_01}}
     {targetBone {boneName=spine_02}}
     {targetBone {boneName=spine_03}}
     {targetBone {boneName=spine_04}}
     {targetBone {boneName=spine_05}}
    ]
    constraintWeights [1,1,1,1,1,1]
   }
   n4 = "Constraint - dyn_gun_constr_base_l (PPPPP)"!(
    ComponentPose=@n3
    Alpha=1
    ConstraintWeights_0=0.0725
    ConstraintWeights_1:float=0.1788
    ConstraintWeights_2:float=0.3594
    ConstraintWeights_3:float=0.2172
    ConstraintWeights_4:float=0.1721
   ) -> (Pose) {
    alphaBoolBlend {blendOption=Linear}
    boneToModify {boneName=dyn_gun_constr_base_l}
    constraintSetup [
     {targetBone {boneName=spine_01}}
     {targetBone {boneName=spine_02}}
     {targetBone {boneName=spine_03}}
     {targetBone {boneName=spine_04}}
     {targetBone {boneName=clavicle_l}}
    ]
    constraintWeights [1,1,1,1,1]
   }
   n5 = "Constraint - dyn_gun_constr_base_r (PPPPP)"!(
    ComponentPose=@n4
    Alpha=1
    ConstraintWeights_0=0.1729
    ConstraintWeights_1:float=0.0722
    ConstraintWeights_2:float=0.1782
    ConstraintWeights_3:float=0.3591
    ConstraintWeights_4:float=0.2176
   ) -> (Pose) {
    alphaBoolBlend {blendOption=Linear}
    boneToModify {boneName=dyn_gun_constr_base_r}
    constraintSetup [
     {targetBone {boneName=clavicle_r}}
     {targetBone {boneName=spine_01}}
     {targetBone {boneName=spine_02}}
     {targetBone {boneName=spine_03}}
     {targetBone {boneName=spine_04}}
    ]
    constraintWeights [1,1,1,1,1]
   }
   n6 = "Constraint - dyn_vest_strap_constr_base_l (PPPPP)"!(
    ComponentPose=@n5
    Alpha=1
    ConstraintWeights_0=0.226
    ConstraintWeights_1:float=0.3631
    ConstraintWeights_2:float=0.2538
    ConstraintWeights_3:float=0.0406
    ConstraintWeights_4:float=0.1164
   ) -> (Pose) {
    alphaBoolBlend {blendOption=Linear}
    boneToModify {boneName=dyn_vest_strap_constr_base_l}
    constraintSetup [
     {targetBone {boneName=spine_01}}
     {targetBone {boneName=spine_02}}
     {targetBone {boneName=spine_03}}
     {targetBone {boneName=spine_04}}
     {targetBone {boneName=pelvis}}
    ]
    constraintWeights [1,1,1,1,1]
   }
   n7 = "Constraint - dyn_legpackstrap_a_base_r (PPP)"!(ComponentPose=@n6, Alpha=1, ConstraintWeights_0=0.0185, ConstraintWeights_1:float=0.0778, ConstraintWeights_2:float=0.9036) -> (Pose) {
    alphaBoolBlend {blendOption=Linear}
    boneToModify {boneName=dyn_legpackstrap_a_base_r}
    constraintSetup [{targetBone {boneName=thigh_r}}, {targetBone {boneName=pelvis}}, {targetBone {boneName=thigh_twist_01_r}}]
    constraintWeights [1,1,1]
   }
   n8 = Anim Dynamics - Bone: dyn_nametag(ComponentPose=@n7, Alpha=1, ExternalForce="0,0,0") {
    angularDampingOverride=1
    angularSpringConstant=1500
    bAngularSpring=true
    bGravityOverrideInSimSpace=true
    bOverrideAngularDamping=true
    gravityScale=0.35
    simulationSpace=RootRelative
    alphaBoolBlend {blendOption=Linear}
    boundBone {boneName=dyn_nametag}
    componentLinearAccScale=<0.5 0.5 0.5>
    componentLinearVelScale=<0.5 0.5 0.5>
    physicsBodyDefinitions [{
      sphereCollisionRadius=0
      boundBone {boneName=dyn_nametag}
      boxExtents=<5 10 10>
      constraintSetup {angularLimitsMax=<0 5 40>, angularLimitsMin=<0 -5 -10>, angularTarget=<0 0 0>}
      localJointOffset=<5 0 0>
     }]
   }
   n9 = Anim Dynamics - Bone: dyn_neck_zipper(ComponentPose=@n8, Alpha=1, ExternalForce="0,0,0") {
    angularDampingOverride=1
    angularSpringConstant=1500
    bAngularSpring=true
    bGravityOverrideInSimSpace=true
    bOverrideAngularDamping=true
    gravityScale=0.35
    simulationSpace=RootRelative
    alphaBoolBlend {blendOption=Linear}
    boundBone {boneName=dyn_neck_zipper}
    componentLinearAccScale=<0.5 0.5 0.5>
    componentLinearVelScale=<0.5 0.5 0.5>
    physicsBodyDefinitions [{
      sphereCollisionRadius=0
      boundBone {boneName=dyn_neck_zipper}
      boxExtents=<5 10 10>
      constraintSetup {angularLimitsMax=<0 0 40>, angularLimitsMin=<0 0 -5>, angularTarget=<0 0 0>}
      localJointOffset=<5 0 0>
     }]
   }
   n10 = RigidBody(ComponentPose=@n9) {
    bClampLinearTranslationLimitToRefPose=true
    bOverrideWorldGravity=true
    overridePhysicsAsset=&/M_MED_Agent_Ace_Physics
    simulationSpace=BaseBoneSpace
    alphaBoolBlend {blendOption=Linear}
    baseBoneRef {boneName=pelvis}
    componentLinearAccScale=<0.4 0.4 0.4>
    componentLinearVelScale=<0.4 0.4 0.4>
    overrideWorldGravity {z=-980}
   }
   Output Pose(Result=Component To Local(ComponentPose=@n10)) {layerGroup=None}
  ]}
 EventGraph {links [], nodes []}
]
variables []
