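# Physics asset for the face accessory part of the medium male Agent Ace cosmetic body, defining physics for its dynamic hair bones.
# Each hair bone and the head has one capsule body, linked by constraints with narrow swing and twist limits.
# Front hair forms a two-link chain from the head, and left strands a–e branch off a main left bone.
# Back, right and remaining left strands attach directly to the head.
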PhysicsAsset /BRCosmetics/Characters/Player/Male/Medium/Bodies/M_MED_Agent_Ace/Meshe/Parts/M_MED_Agent_Ace_FaceAcc_Physics
bodies {
 dyn_hair_a_l [{shapeType=Capsule, length=1.3, radius=1.3, center=<0 0 -2>}]
 dyn_hair_a_r [{shapeType=Capsule, length=1.3, radius=1.3, center=<0 0 -2.5>}]
 dyn_hair_b_l [{shapeType=Capsule, length=1.3, radius=1.3, center=<0 0 -2>}]
 dyn_hair_b_r [{shapeType=Capsule, length=1.3, radius=1.3, center=<0 0 -2.5>}]
 dyn_hair_bk_a [{shapeType=Capsule, length=1.3, radius=1.3, center=<0 0 -2>}]
 dyn_hair_bk_b [{shapeType=Capsule, length=1.3, radius=1.3, center=<0 0 -2>}]
 dyn_hair_bk_c [{shapeType=Capsule, length=1.25, radius=2.5, center=<0 0 -4.2609>}]
 dyn_hair_c_l [{shapeType=Capsule, length=1.3, radius=1.3, center=<0 0 -2>}]
 dyn_hair_c_r [{shapeType=Capsule, length=1.3, radius=1.3, center=<0 0 -2.5>}]
 dyn_hair_d_l [{shapeType=Capsule, length=1.3, radius=1.3, center=<0 0 -2>}]
 dyn_hair_d_r [{shapeType=Capsule, length=1.3, radius=1.3, center=<0 0 -2.5>}]
 dyn_hair_e_l [{shapeType=Capsule, length=1.3, radius=1.3, center=<0 0 -2>}]
 dyn_hair_f_l [{shapeType=Capsule, length=1.3, radius=1.3, center=<0 0 -2>}]
 dyn_hair_front_a [{shapeType=Capsule, length=1.5, radius=1.25, center=<0 0 -1.6442>}]
 dyn_hair_front_b [{shapeType=Capsule, length=1.5, radius=1.25, center=<0 0 -1.6442>}]
 dyn_hair_g_l [{shapeType=Capsule, length=1.3, radius=1.3, center=<0 0 -2>}]
 dyn_hair_main_l [{shapeType=Capsule, length=2.75, radius=2, center=<0 0 -2.3383>}]
 head [{shapeType=Capsule, length=1, radius=1}]
}
constraints [
 {bone1Name=dyn_hair_front_a, bone2Name=head, swing1LimitDegrees=10, swing2LimitDegrees=15, twistLimitDegrees=5}
 {bone1Name=dyn_hair_front_b, bone2Name=dyn_hair_front_a, swing1LimitDegrees=10, swing2LimitDegrees=20, twistLimitDegrees=5}
 {bone1Name=dyn_hair_main_l, bone2Name=head, swing1LimitDegrees=15, swing2LimitDegrees=15, twistLimitDegrees=10}
 {bone1Name=dyn_hair_a_l, bone2Name=dyn_hair_main_l, swing1LimitDegrees=20, swing2LimitDegrees=20, twistLimitDegrees=5}
 {bone1Name=dyn_hair_e_l, bone2Name=dyn_hair_main_l, swing1LimitDegrees=20, swing2LimitDegrees=20, twistLimitDegrees=5}
 {bone1Name=dyn_hair_d_l, bone2Name=dyn_hair_main_l, swing1LimitDegrees=20, swing2LimitDegrees=20, twistLimitDegrees=5}
 {bone1Name=dyn_hair_b_l, bone2Name=dyn_hair_main_l, swing1LimitDegrees=20, swing2LimitDegrees=20, twistLimitDegrees=5}
 {bone1Name=dyn_hair_c_l, bone2Name=dyn_hair_main_l, swing1LimitDegrees=20, swing2LimitDegrees=20, twistLimitDegrees=5}
 {bone1Name=dyn_hair_bk_c, bone2Name=head, swing1LimitDegrees=20, swing2LimitDegrees=5, twistLimitDegrees=20}
 {bone1Name=dyn_hair_bk_b, bone2Name=head, swing1LimitDegrees=5, swing2LimitDegrees=20, twistLimitDegrees=20}
 {bone1Name=dyn_hair_bk_a, bone2Name=head, swing1LimitDegrees=5, swing2LimitDegrees=20, twistLimitDegrees=20}
 {bone1Name=dyn_hair_d_r, bone2Name=head, swing1LimitDegrees=25, swing2LimitDegrees=25, twistLimitDegrees=20}
 {bone1Name=dyn_hair_c_r, bone2Name=head, swing1LimitDegrees=25, swing2LimitDegrees=25, twistLimitDegrees=20}
 {bone1Name=dyn_hair_a_r, bone2Name=head, swing1LimitDegrees=25, swing2LimitDegrees=25, twistLimitDegrees=20}
 {bone1Name=dyn_hair_b_r, bone2Name=head, swing1LimitDegrees=25, swing2LimitDegrees=25, twistLimitDegrees=20}
 {bone1Name=dyn_hair_f_l, bone2Name=head, swing1LimitDegrees=20, swing2LimitDegrees=20, twistLimitDegrees=20}
 {bone1Name=dyn_hair_g_l, bone2Name=head, swing1LimitDegrees=20, swing2LimitDegrees=20, twistLimitDegrees=20}
]
